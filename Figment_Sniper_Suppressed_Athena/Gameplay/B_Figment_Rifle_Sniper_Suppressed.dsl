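; Weapon Blueprint for the Figment suppressed sniper rifle that adjusts its muzzle effect to show a suppressor.
; On weapon fire FX, it runs the parent logic, then casts the spawned muzzle system to a Niagara component.
; It sets that component's "isMuzzleSupressor" bool (spelled with one "p") to true after each parent call.
; The construction script only calls its parent.
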(event EventOnPlayWeaponFireFX
  (params (bPersistentFire :type Boolean) (bSecondaryFire :type Boolean))
  (|CallParentFunction bPersistentFire bSecondaryFire)
  (Niagara|SetNiagaraVariable(Bool) (Utilities|Casting|CastToNiagaraComponent (Variables|Effects|Muzzle|GetSpawned%20Muzzle%20System)) "isMuzzleSupressor" true))

(fn ConstructionScript
  (|CallParentFunction))
